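# Quest sound cue for the Alien Detector's billboard alien, combining a continuous loop with a one-shot unhide sound.
# A Mixer plays the looping alien wave alongside a quieter Billboard_Unhide wave, which a Modulator pitches down by a fixed amount.
# The whole cue is slightly pitched down, routes to the quests sound class, and uses its own attenuation override in place of the referenced supply-drop marker attenuation asset.
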SoundCue /Game/Athena/Sounds/Quests/AlienDetector/AlienDetector_BillboardAlien_Loop_Cue
root=Mixer_0
nodes [
 {id=Mixer_0, children [WavePlayer_1,Modulator_0], properties {inputVolume [1,0.6]}}
 {id=Modulator_0, children [WavePlayer_0], properties {pitchMax=0.6, pitchMin=0.6}}
 {id=WavePlayer_0, properties {soundWaveAssetPtr=&/AlienDetector_Billboard_Unhide_01}}
 {id=WavePlayer_1, properties {bLooping=true, soundWaveAssetPtr=&/AlienDetector_BillboardAlien_Loop_01}}
]
properties {
 attenuationSettings=/Game/Sounds/Foley_Loot/Containers/SupplyDrops/Fort_SupplyDrop_Marker_Close_Athena_Att
 bOverrideAttenuation=true
 maxDistance=900
 pitchMultiplier=0.9
 soundClassObject=&Fort_Quests
 volumeMultiplier=1
 attenuationOverrides {falloffDistance=800, stereoSpread=600, attenuationShapeExtents {x=100}}
}
