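# Sound cue for the looping billboard sound of the Alien Detector quest, in its "Normal" variant.
# A single looping wave player plays one billboard loop wave at reduced volume in the quests sound class.
# It references the supply drop marker close attenuation asset but overrides attenuation with its own falloff distance, stereo spread and shape extents.
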SoundCue /Game/Athena/Sounds/Quests/AlienDetector/AlienDetector_BillboardNormal_Loop_Cue
root=WavePlayer_0
nodes [{
  id=WavePlayer_0
  properties {bLooping=true, soundWaveAssetPtr=&/AlienDetector_BillboardNormal_Loop_01}
 }]
properties {
 attenuationSettings=/Game/Sounds/Foley_Loot/Containers/SupplyDrops/Fort_SupplyDrop_Marker_Close_Athena_Att
 bOverrideAttenuation=true
 maxDistance=1000
 soundClassObject=&Fort_Quests
 volumeMultiplier=0.7
 attenuationOverrides {falloffDistance=800, stereoSpread=600, attenuationShapeExtents {x=200}}
}
