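# Sound cue for the Alien Detector quest case on the ground, playing a spatial loop in the quest sound class.
# A mixer starts a looping ground wave together with a one-shot case-close wave at equal volume, so the close sound plays only at the start.
# The cue overrides attenuation with its own falloff rather than using the referenced supply drop marker settings, and slightly raises pitch while lowering volume.
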SoundCue /Game/Athena/Sounds/Quests/AlienDetector/AlienDetector_CaseGround_Loop_Cue
root=Mixer_0
nodes [
 {id=Mixer_0, children [WavePlayer_0,WavePlayer_1], properties {inputVolume [1,1]}}
 {id=WavePlayer_0, properties {bLooping=true, soundWaveAssetPtr=&/AlienDetector_CaseGround_Loop_01}}
 {id=WavePlayer_1, properties {soundWaveAssetPtr=&/AlienDetector_Case_Close_01}}
]
properties {
 attenuationSettings=/Game/Sounds/Foley_Loot/Containers/SupplyDrops/Fort_SupplyDrop_Marker_Close_Athena_Att
 bOverrideAttenuation=true
 maxDistance=1000
 pitchMultiplier=1.1
 soundClassObject=&Fort_Quests
 volumeMultiplier=0.7
 attenuationOverrides {falloffDistance=800, attenuationShapeExtents {x=200}}
}
