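# Sound cue for the success sound of the Alien Detector case, filed under Athena quest sounds.
# A single wave feeds a default Modulator for slight pitch and volume variation, played slightly lowered in pitch through a loot-reward sound class.
# Overridden attenuation uses a custom falloff curve, distance low-pass, occlusion and interpolated listener focus that boosts on-axis volume.
# Overridden concurrency stops the quietest instance when its limit is reached.
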SoundCue /Game/Athena/Sounds/Quests/AlienDetector/AlienDetector_Case_Success_Cue
root=Modulator_1
nodes [
 {id=Modulator_1, children [WavePlayer_1]}
 {id=WavePlayer_1, properties {soundWaveAssetPtr=&/AlienDetector_Case_Success_01}}
]
properties {
 bOverrideAttenuation=true
 bOverrideConcurrency=true
 maxDistance=1200
 pitchMultiplier=0.9
 priority=50
 soundClassObject=&Fort_Loot_Reward_PSM
 volumeMultiplier=1
 attenuationOverrides {
  bEnableFocusInterpolation=true
  bEnableListenerFocus=true
  falloffDistance=1100
  focusAttackInterpSpeed=20
  focusAzimuth=20
  focusReleaseInterpSpeed=10
  focusVolumeAttenuation=1.75
  lPFFrequencyAtMax=10000
  lPFRadiusMax=1000
  lPFRadiusMin=500
  nonFocusAzimuth=40
  nonFocusVolumeAttenuation=0.8
  occlusionInterpolationTime=0.3
  occlusionLowPassFilterFrequency=3000
  occlusionVolumeAttenuation=0.75
  stereoSpread=4000
  attenuationShapeExtents {x=100}
  customAttenuationCurve {editorCurveData {keys [<at 0 1 RCIM_Cubic>, {arriveTangent=-1, interpMode=RCIM_Cubic, leaveTangent=-1, time=0.5, value=0.1}, <at 1 0 RCIM_Cubic>]}}
 }
 concurrencyOverrides {resolutionRule=StopQuietest}
}
